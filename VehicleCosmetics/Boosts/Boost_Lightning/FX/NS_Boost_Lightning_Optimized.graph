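# Niagara system for a lightning-themed vehicle boost cosmetic, with paint color variants, car offset and separate tuning for frontend preview and game modes.
# A multi-renderer ribbon emitter streams velocity-driven trails from each boost port socket.
# Two sprite emitters add flipbook lightning at the same skeletal mesh sockets: flung, spinning bolts and a source flash.
# The source flash merges two layers into one emitter through spawn-group masking of size, color and alpha.
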NiagaraSystem /VehicleCosmetics/Boosts/Boost_Lightning/FX/NS_Boost_Lightning_Optimized
dependencies {
 usedDataInterfaces [
  &NiagaraDataInterfaceCurve
  &NiagaraDataInterfaceVector2DCurve
  &NiagaraDataInterfaceSpriteRendererInfo
  &NiagaraDataInterfaceSkeletalMesh
 ]
 usedDynamicInputs []
 usedModules [
  &self:SystemSpawnScript.SystemScriptSource.SystemScriptGraph.NiagaraNodeAssignment_0.SetVariables_#1
  &self:SystemSpawnScript.SystemScriptSource.SystemScriptGraph.NiagaraNodeAssignment_1.SetVariables_#2
  &SystemState
 ]
}
emitters {
 Lightning_Trail001 {
  emitterUpdateScript [
   &EmitterState(Life Cycle Mode=::NewEnumerator1, Loop Duration=1, UseLoopDelay=false)
   &VC_FilteredSocketCount()
   &VC_RateQualityHandler()
   &VC_SpawnRibbon_v1(
    Number of Ribbons=$Emitter.PortCount
    SpawnRate=&MakeCustomFloatFromBool(Bool=$User.bIsFrontendPreview, True Float=60, False Float=200)
   )
  ]
  particleSpawnScript [
   &VC_PaintedVariantsOverride(
    BurntSienna=/VehicleCosmetics/FX/Niagara/Modules/VC_Mask_LinearColor(
     Coro Color:LinearColor=<rgba 10 2.0877 0.6957 1>
     Coro FrontEnd Color:LinearColor=<rgba 15 3.1316 1.0435 1>
     FN Color:LinearColor=<rgba 30 6.2632 2.087 1>
    )
    Cobalt=&VC_Mask_LinearColor(Coro Color:LinearColor=<rgba 1 1.45 10 1>, Coro FrontEnd Color:LinearColor=<rgba 2.5 3.625 25 1>, FN Color:LinearColor=<rgba 5 7.2499 50 1>)
    Crimson=&VC_Mask_LinearColor(Coro Color:LinearColor=<rgba 10 0 0.1667 1>, Coro FrontEnd Color:LinearColor=<rgba 20 0 0.3333 1>, FN Color:LinearColor=<rgba 50 0 0.8333 1>)
    DefaultColor=&VC_Mask_LinearColor(
     Coro Color:LinearColor=<rgba 2.56 6.8763 10 1>
     Coro FrontEnd Color:LinearColor=<rgba 1.28 3.4382 5 1>
     FN Color:LinearColor=<rgba 12.8 34.3816 50 1>
    )
    ForestGreen=&VC_Mask_LinearColor(Coro Color:LinearColor=<rgba 0 10 0.4614 1>, Coro FrontEnd Color:LinearColor=<rgba 0 13 0.5998 1>, FN Color:LinearColor=<rgba 0 40 1.8455 1>)
    Grey=<rgba 4 4 4 1>
    Lime=&VC_Mask_LinearColor(
     Coro Color:LinearColor=<rgba 6.6667 10 0 1>
     Coro FrontEnd Color:LinearColor=<rgba 6.6667 10 0 1>
     FN Color:LinearColor=<rgba 26.6667 40 0 1>
    )
    Orange=&VC_Mask_LinearColor(
     Coro Color:LinearColor=<rgba 10 3.6667 0.5 1>
     Coro FrontEnd Color:LinearColor=<rgba 15 5.5 0.75 1>
     FN Color:LinearColor=<rgba 40 14.6667 2 1>
    )
    Pink=&VC_Mask_LinearColor(
     Coro Color:LinearColor=<rgba 15 3.0233 8.958 1>
     Coro FrontEnd Color:LinearColor=<rgba 15 3.0233 8.958 1>
     FN Color:LinearColor=<rgba 30 6.0467 17.916 1>
    )
    Purple=&VC_Mask_LinearColor(
     Coro Color:LinearColor=<rgba 4.3333 0 10 2.5>
     Coro FrontEnd Color:LinearColor=<rgba 4.3333 0 10 2.5>
     FN Color:LinearColor=<rgba 8.6667 0 20 2.5>
    )
    Saffron=&VC_Mask_LinearColor(Coro Color:LinearColor=<rgba 10 5.8333 0 1>, Coro FrontEnd Color:LinearColor=<rgba 12 7 0 1>, FN Color:LinearColor=<rgba 20 11.6667 0 1>)
    SkyBlue=&VC_Mask_LinearColor(
     Coro Color:LinearColor=<rgba 0 6.6539 10 1>
     Coro FrontEnd Color:LinearColor=<rgba 0 9.9808 15 1>
     FN Color:LinearColor=<rgba 0 19.9615 30 1>
    )
    TitaniumWhite=&VC_Mask_LinearColor(Coro Color:LinearColor=<rgba 10 10 10 1>, Coro FrontEnd Color:LinearColor=<rgba 7 7 7 1>, FN Color:LinearColor=<rgba 100 100 100 1>)
   )
   &VC_InitializeRibbon_v1(
    Color=$Particles.Color
    Custom Offset=true
    Enable Skeletal Position=true
    Lifetime=&Multiply_Float(
     A=&MakeCustomFloatFromBool(Bool=$User.bIsFrontendPreview, True Float=0.35, False Float=0.2)
     B=/VehicleCosmetics/FX/Niagara/Modules/VC_Mask_Floats(Coro Value:NiagaraFloat=3, FrontEnd Value:NiagaraFloat=1.5, FN Value:NiagaraFloat=1.5)
    )
    Local Vector Offset=&AddVectorToPosition(
     Position=&MakeCustomVectorFromBool(Bool=true, True Vector=<15 0 0>, False Vector=<50 0 0>)
     Vector=$User.DTCarOffset
    )
    Offset Amount=1
    Ribbon Width:NiagaraFloat=&MakeCustomFloatFromBool(Bool=$User.bIsFrontendPreview, True Float=120, False Float=150)
   )
   &AddVelocity(Velocity=&VC_Mask_Vectors(
     Coro Value:Vector3f=<-600 0 0>
     FrontEnd Value:Vector3f=&RandomRangeVector(Minimum=<-3150 0 0>, Maximum=<-3300 0 0>)
     FN Value:Vector3f=<-4000 0 0>
    )) {module=AddVelocity001}
   /VehicleCosmetics/FX/VC_FXParents/Modules/DLR_ProductPreviews(GameModeID:NiagaraInt32=$System.GameModeID, bPreviewVehicleSpeed:NiagaraBool=$System.bPreviewVehicleSpeed) {enabled=false}
  ]
  particleUpdateScript [
   &ParticleState()
   &InheritSourceMovement(
    Applied Position Scale=&Multiply_VectorByFloat(
     Vector=<1 1 1>
     Float=&FloatFromCurve(
      FloatCurve=DataInterface() {propertyValues {Curve {keys [
          {arriveTangent=-1.819, interpMode=RCIM_Cubic, leaveTangent=-1.819, tangentMode=RCTM_User, time=0.1, value=1}
          {interpMode=RCIM_Cubic, tangentMode=RCTM_User, time=1, value=-1}
         ]}}}
      Scale Curve=&VC_Mask_Floats(Coro Value:NiagaraFloat=0.25, FrontEnd Value:NiagaraFloat=1, FN Value:NiagaraFloat=1)
     )
    )
    ApplyForce=false
    ApplyPosition=true
    ApplyPositionFalloff=false
   ) {enabled=false, module=InheritSourceMovement001}
   &InheritSourceMovement(
    Applied Position Scale=&VectorFromFloat(Value=&FloatFromCurve(FloatCurve=DataInterface() {propertyValues {Curve {
         defaultValue=1
         keys [
          {arriveTangent=-1.7885, interpMode=RCIM_Cubic, leaveTangent=-1.7885, tangentMode=RCTM_User, time=0.05, value=1}
          {interpMode=RCIM_Cubic, tangentMode=RCTM_User, time=1, value=0}
         ]
        }}}))
    ApplyForce=false
    ApplyPosition=true
    ApplyPositionFalloff=false
   )
   &SolveForcesAndVelocity(Clamp Velocity=false, Limit Acceleration=false)
   &ScaleRibbonWidth(Ribbon Width Scale=&MakeCustomFloatFromBool(
     Bool=$User.bIsFrontendPreview
     True Float=&MakeCustomFloatFromBool(
      Bool=$User.bIsPainted
      True Float=&FloatFromCurve(FloatCurve=DataInterface() {propertyValues {Curve {keys [<at 0 1>, <at 0.3 1.5>, <at 0.8 1>, {arriveTangent=-6.693, leaveTangent=-6.693, time=1, value=0.1}]}}})
      False Float=&FloatFromCurve(FloatCurve=DataInterface() {propertyValues {Curve {keys [<at 0 0.5>, <at 0.1 1>, <at 0.3 1.3>, <at 0.8 1>, {arriveTangent=-6.693, leaveTangent=-6.693, time=1, value=0.1}]}}})
     )
     False Float=&FloatFromCurve(FloatCurve=DataInterface() {propertyValues {Curve {keys [
          {arriveTangent=5.3261, interpMode=RCIM_Cubic, leaveTangent=5.3261, tangentMode=RCTM_User, time=0, value=0}
          {interpMode=RCIM_Cubic, tangentMode=RCTM_User, time=0.5, value=1}
          {tangentMode=RCTM_User, time=1, value=0}
         ]}}})
    )) {module=ScaleRibbonWidth001}
   &ScaleColor(
    Scale Alpha=&MakeCustomFloatFromBool(
     Bool=$User.bIsFrontendPreview
     True Float=&MakeCustomFloatFromBool(
      Bool=$User.bIsPainted
      True Float=1
      False Float=&FloatFromCurve(FloatCurve=DataInterface() {propertyValues {Curve {keys [
           {interpMode=RCIM_Cubic, tangentMode=RCTM_User, time=0, value=0.5}
           {interpMode=RCIM_Cubic, tangentMode=RCTM_User, time=0.5, value=1}
           {arriveTangent=-2.0024, tangentMode=RCTM_User, time=0.8, value=0.5}
           <at 0.93 0>
          ]}}})
     )
     False Float=&FloatFromCurve(FloatCurve=DataInterface() {propertyValues {Curve {keys [
          {arriveTangent=0.0212, interpMode=RCIM_Cubic, leaveTangent=0.0212, tangentMode=RCTM_User, time=0, value=0.1}
          <at 0.5 0.98 RCIM_Cubic>
          {arriveTangent=0.0665, interpMode=RCIM_Cubic, leaveTangent=0.0665, tangentMode=RCTM_User, time=1, value=0.1}
         ]}}})
    )
    Scale RGB=&Multiply_VectorByFloat(
     Vector=<1 1 1>
     Float=&MakeCustomFloatFromBool(
      Bool=$User.bIsFrontendPreview
      True Float=&MakeCustomFloatFromBool(
       Bool=$User.bIsPainted
       True Float=&FloatFromCurve(
        FloatCurve=DataInterface() {propertyValues {Curve {keys [
            {arriveTangent=-1.0176, interpMode=RCIM_Cubic, leaveTangent=-1.0176, tangentMode=RCTM_User, time=0, value=0.25}
            {arriveTangent=0.0018, interpMode=RCIM_Cubic, leaveTangent=0.0018, tangentMode=RCTM_User, time=0.1, value=1}
            {arriveTangent=-0.5556, interpMode=RCIM_Cubic, leaveTangent=-0.5556, time=0.5, value=0.7}
            {arriveTangent=-0.1054, interpMode=RCIM_Cubic, leaveTangent=-0.1054, tangentMode=RCTM_User, time=1, value=0.5}
           ]}}}
        Scale Curve=10
       )
       False Float=&FloatFromCurve(
        FloatCurve=DataInterface() {propertyValues {Curve {keys [
            {arriveTangent=0.0214, interpMode=RCIM_Cubic, leaveTangent=0.0214, tangentMode=RCTM_User, time=0.05, value=0}
            {arriveTangent=2.3098, interpMode=RCIM_Cubic, leaveTangent=2.3098, tangentMode=RCTM_User, time=0.3, value=1}
            {arriveTangent=-3.2531, interpMode=RCIM_Cubic, leaveTangent=-3.2531, tangentMode=RCTM_User, time=0.8, value=0.8}
            {arriveTangent=-0.0165, interpMode=RCIM_Cubic, leaveTangent=-0.0165, tangentMode=RCTM_User, time=1, value=0}
           ]}}}
        Scale Curve=20
       )
      )
      False Float=&FloatFromCurve(
       FloatCurve=DataInterface() {propertyValues {Curve {keys [
           {arriveTangent=-0.097, interpMode=RCIM_Cubic, leaveTangent=-0.097, tangentMode=RCTM_User, time=0, value=2}
           <at 0.1 3 RCIM_Cubic>
           <at 0.25 2 RCIM_Cubic>
          ]}}}
       Scale Curve=&MakeCustomFloatFromBool(Bool=$User.bIsPainted, True Float=30, False Float=5)
      )
     )
    )
    ScaleA=true
    ScaleRGB=true
   )
   /SetVariables(
    Particles.R2 Width:NiagaraFloat=&Multiply_Float(A=$Particles.RibbonWidth, B=2.6)
    Particles.float:NiagaraFloat=&Multiply_Float(A=$Particles.RibbonWidth, B=3)
   ) {isSetParameters=true}
   /SetVariables(Particles.Linear Color=&MultiplyLinearColor(Linear Color=$Particles.Color, Scale Factor=<0.1 0.1 0.1 0.06>)) {isSetParameters=true}
   &VC_UpdateRibbon_v1(Enable Skeletal Position=true)
  ]
  rendererClasses [&NiagaraRibbonRendererProperties]
  renderers [
   {rendererClass=&NiagaraRibbonRendererProperties}
   {rendererClass=&NiagaraRibbonRendererProperties, rendererIndex=1}
   {rendererClass=&NiagaraRibbonRendererProperties, rendererIndex=2}
   {rendererClass=&NiagaraRibbonRendererProperties, rendererIndex=3}
  ]
 }
 NE_Lightning_template001 {
  emitterUpdateScript [
   &EmitterState(Life Cycle Mode=::NewEnumerator1, Loop Duration=0.1, Recalculate Duration Each Loop=true, UseLoopDelay=false)
   &VC_FilteredSocketCount()
   &SpawnRate(
    SpawnRate=&MakeCustomFloatFromBool(Bool=$User.bIsFrontendPreview, True Float=40, False Float=50)
    Use Spawn Probability=false
   )
  ]
  particleSpawnScript [
   &VC_PaintedVariantsOverride(
    BurntSienna=<rgba 30 3.9352 0 1>
    Cobalt=<rgba 0 0 30 0.8>
    Crimson=<rgba 30 1.1864 1.1864 1>
    DefaultColor=<rgba 0.3333 0.8 2 1>
    ForestGreen=<rgba 0 30 5 1>
    Grey=<rgba 30 30 30 0.5>
    Lime=<rgba 17.5047 30 0 1>
    Orange=<rgba 30 10.546 0 0.502>
    Pink=<rgba 30 9.0163 26.8881 1>
    Purple=<rgba 15.0866 0 30 1>
    Saffron=<rgba 30 15 0 0.35>
    SkyBlue=<rgba 0 19.9112 30 0.75>
    TitaniumWhite=&VC_Mask_LinearColor(Coro Color:LinearColor=<rgba 15 15 15 0.5>, Coro FrontEnd Color:LinearColor=<rgba 15 15 15 0.5>, FN Color:LinearColor=<rgba 30 30 30 0.5>)
   )
   &InitializeParticle(
    Color Mode=::NewEnumerator0
    Lifetime=&RandomRangeFloat(Minimum=0.2, Maximum=0.2)
    Position Offset=&AddVectorToPosition(
     Position=&MakeCustomVectorFromBool(Bool=$User.bIsFrontendPreview, True Vector=<0 0 -12.5>, False Vector=<0 0 0>)
     Vector=$User.DTCarOffset
    )
    Position Offset Coordinate Space=::Local
    Sprite Rotation Mode=::NewEnumerator2
    Sprite Size=&Multiply_Vector2D(A=<30 25>, B=&RandomRangeVector2D(Minimum=<0.15 0.2>, Maximum=<1 1.1>))
    Sprite Size Mode=::NewEnumerator4
    UsePositionOffset=true
    Write Material Random=true
   )
   &PartitionParticles(Number Of Partitions=$Emitter.PortCount)
   &SkeletalMeshLocation(
    Bone / Socket Index=$Particles.PartitionParticles.Partition
    Bone / Socket Sampling Mode=::NewEnumerator5
    MaskSpawnGroup=false
    Mesh Sampling Type=::NewEnumerator6
    Sampled Position Offset=&AddVectorToPosition(
     Position=/Niagara/DynamicInputs/TypeConversions/MakeCustomPositionFromBool(Bool:NiagaraBool=$User.bIsFrontendPreview, True Position:NiagaraPosition=<-20 0 12.5>, False Position:NiagaraPosition=<-10 0 0>)
     Vector=$User.DTCarOffset
    )
    Sampled Position Offset Coordinate Space:ENiagaraSkelMeshTransforms=::NewEnumerator2
    Skeletal Mesh=$System.SkeletalMesh
   )
   &AddVelocity(
    Velocity=&MakeCustomVectorFromBool(Bool=$User.bIsFrontendPreview, True Vector=<2200 0 0>, False Vector=<1000 0 0>)
    Velocity Speed Scale=&RandomRangeFloat(Minimum=-1, Maximum=-1.2)
   )
   &ShapeLocation(MaskSpawnGroup=false, Override Seed=false, Sphere Radius=50, UseOwnerScale=true)
  ]
  particleUpdateScript [
   &ParticleState()
   &SpriteRotationRate(Rotation Rate=&RandomRangeFloat(Minimum=-100, Maximum=100))
   &InheritSourceMovement(
    Applied Position Scale=&Multiply_VectorByFloat(
     Vector=<1 1 1>
     Float=&MakeCustomFloatFromBool(
      Bool=$User.bIsFrontendPreview
      True Float=0
      False Float=&FloatFromCurve(FloatCurve=DataInterface() {propertyValues {Curve {keys [
           {arriveTangent=-0.3477, interpMode=RCIM_Cubic, leaveTangent=-0.3477, tangentMode=RCTM_User, time=0.3, value=1}
           {arriveTangent=-3.3085, interpMode=RCIM_Cubic, leaveTangent=-3.3085, tangentMode=RCTM_User, time=1, value=0.1}
          ]}}})
     )
    )
    ApplyForce=false
    ApplyPosition=true
    ApplyPositionFalloff=false
   )
   &ScaleSpriteSize(
    Uniform Curve Scale=2.25
    Uniform Curve Sprite Scale=DataInterface() {propertyValues {Curve {keys [
        {interpMode=RCIM_Cubic, tangentMode=RCTM_User, time=0, value=1}
        {interpMode=RCIM_Cubic, tangentMode=RCTM_User, time=0.2, value=1}
        {interpMode=RCIM_Cubic, tangentMode=RCTM_User, time=1, value=1.25}
       ]}}}
   )
   &SolveForcesAndVelocity(Clamp Velocity=false, Limit Acceleration=false)
   &DynamicMaterialParameters(
    Index 0 Param 1=&FloatFromCurve(
     FloatCurve=DataInterface() {propertyValues {Curve {
        defaultValue=0
        keys [
         {interpMode=RCIM_Cubic, tangentMode=RCTM_User, time=0, value=0}
         {interpMode=RCIM_Cubic, tangentMode=RCTM_User, time=0.25, value=1}
         <at 0.5 0 RCIM_Cubic>
         <at 0.75 1 RCIM_Cubic>
         <at 1 0>
        ]
       }}}
     Scale Curve=&RandomRangeFloat(Minimum=0, Maximum=0.1)
    )
    Module.Index0Float0Write=true
    Module.Index0Float1Write=true
    Module.Index0Float2Write=true
    Module.Index0Float3Write=true
    Module.Index1Float0Write=true
    Module.Index1Float1Write=true
    Module.Index1Float2Write=true
    Module.Index1Float3Write=true
    Write Parameter Index 1=::NewEnumerator1
   ) {module=DynamicMaterialParameters001}
   &SubUVAnimation(
    Random Start Frame=true
    "Sprite Renderer "=DataInterface() {propertyValues {SpriteRenderer=&self:NE_Lightning_template001_0.NiagaraSpriteRendererProperties_0}}
    Start Frame Offset:NiagaraInt32=&RandomRangeInt(Minimum=0, Maximum=10, Evaluation Type=::NewEnumerator1)
    SubUV Animation Mode:ENiagara_SubUVLookupModeV2=::NewEnumerator1
    SubUV Curve=DataInterface() {propertyValues {Curve {defaultValue=0, keys [{tangentMode=RCTM_User, time=0, value=0}, {tangentMode=RCTM_User, time=1, value=1}]}}}
    UseEndFrame=false
    UseStartFrame=false
   ) {module=SubUVAnimation001}
   &ScaleColor(
    Scale Alpha=&MakeCustomFloatFromBool(Bool=$User.bIsPainted, True Float=1, False Float=1)
    Scale RGB=&Multiply_VectorByFloat(
     Vector=<1 1 1>
     Float=&MakeCustomFloatFromBool(Bool=$User.bIsPainted, True Float=0.5, False Float=&RandomRangeFloat(Minimum=10, Maximum=20))
    )
    ScaleA=true
    ScaleRGB=true
   ) {module=ScaleColor001}
  ]
  rendererClasses [&NiagaraSpriteRendererProperties]
  renderers [{rendererClass=&NiagaraSpriteRendererProperties}]
 }
 NE_Source_Lightning001 {
  emitterUpdateScript [
   &EmitterState(Life Cycle Mode=::NewEnumerator1, Loop Duration=1, Recalculate Duration Each Loop=true, UseLoopDelay=false)
   &VC_FilteredSocketCount()
   &SpawnRate(SpawnRate=30, Use Spawn Probability=false)
   &SpawnRate(SpawnGroup=1, SpawnRate=25, Use Spawn Probability=false) {module=SpawnRate001}
  ]
  particleSpawnScript [
   &VC_PaintedVariantsOverride(
    BurntSienna=<rgba 2 0.284 0.02 1>
    Cobalt=<rgba 0 0 5 1>
    Crimson=&VC_Mask_LinearColor(
     Coro Color:LinearColor=<rgba 5 0.0999 0.1816 1>
     Coro FrontEnd Color:LinearColor=<rgba 2 0.04 0.0726 1>
     FN Color:LinearColor=<rgba 5 0.0999 0.1816 1>
    )
    DefaultColor=<rgba 0.5 1.2 3 1>
    ForestGreen=<rgba 0 1.5 0.0678 1>
    Grey=<rgba 0.5 0.5 0.5 1>
    Lime=<rgba 1.3333 2 0 0.5>
    Orange=<rgba 2 0.6308 0 1>
    Pink=<rgba 2 0.402 1 1>
    Purple=<rgba 2 0 4 1>
    R1 DefaultColor=<rgba 1.3333 3.2 8 1>
    Saffron=&VC_Mask_LinearColor(
     Coro Color:LinearColor=<rgba 15 6.25 0 0.3>
     Coro FrontEnd Color:LinearColor=<rgba 5 2.0833 0 0.3>
     FN Color:LinearColor=<rgba 15 6.25 0 0.3>
    )
    SkyBlue=<rgba 0 1.3 2 1>
    TitaniumWhite=<rgba 1.5 1.5 1.5 0.502>
    Use R1 Painted Overrides=true
   )
   /SetVariables(Particles.VisTag:NiagaraInt32=&SpawnGroup()) {isSetParameters=true}
   &InitializeParticle(
    Color Mode=::NewEnumerator0
    Lifetime=&MakeCustomFloatFromBool(Bool=$User.bIsFrontendPreview, True Float=&RandomRangeFloat(Minimum=0.2, Maximum=0.4), False Float=0.4)
    Sprite Rotation Angle=&MaskFloat(Mask Source=0, Mask Parameter=?, Mask=0, Masked Float=90, Passthrough Float=1)
    Sprite Rotation Mode=::NewEnumerator0
    Sprite Size=&MaskVector2dBySpawnGroup(
     Masked Vector2D=&RandomRangeVector2D(Minimum=<175 105>, Maximum=<150 105>)
     Passthrough Vector2D=<65 80>
     Spawn Group Mask=0
    )
    Sprite Size Mode=::NewEnumerator4
    UsePositionOffset=false
    Write Material Random=true
   )
   &PartitionParticles(Number Of Partitions=$Emitter.PortCount, Particles Per Partition=3, Partition Mode:ENiagara_PartitionMode=::NewEnumerator1)
   &SkeletalMeshLocation(
    Bone / Socket Index=$Particles.PartitionParticles.Partition
    Bone / Socket Sampling Mode=::NewEnumerator5
    MaskSpawnGroup=false
    Mesh Sampling Type=::NewEnumerator6
    Sampled Position Offset=&AddVectorToPosition(
     Position=&MaskVectorBySpawnGroup(
      Masked Vector=&MakeCustomVectorFromBool(
       Bool=$User.bIsFrontendPreview
       True Vector=&AddVectorToPosition(Position=<-50 0 0>, Vector=<0 0 0>)
       False Vector=&AddVectorToPosition(Position=<-50 0 0>, Vector=<0 0 0>)
      )
      Passthrough Vector=<-60 0 0>
     )
     Vector=$User.DTCarOffset
    )
    Skeletal Mesh=$System.SkeletalMesh
    Use Sampled Mesh Attribute as Ribbon ID=false
   )
   &AddVelocity(Velocity=<100 0 0>, Velocity Speed Scale=&MaskFloatBySpawnGroup(Masked Float=0, Passthrough Float=1))
   &ApplyInitialForces(Apply Force to Position=true)
   &ShapeLocation(
    Box / Plane Mode:ENiagara_BoxPlaneMode=::NewEnumerator1
    MaskSpawnGroup=false
    Non Uniform Scale=&MakeCustomVectorFromBool(Bool=$User.bIsFrontendPreview, True Vector=<0.5 0.5 0.5>, False Vector=<1.5 0.5 0.5>)
    Override Seed=false
    Plane Size:Vector2f=<0 100>
    Shape Primitive=::NewEnumerator2
    UseOwnerScale=true
   )
   &AlignSpriteToMeshOrientation(
    Mesh Orientation Relative Sprite Alignment Vector=&MaskVectorBySpawnGroup(Masked Vector=<-1 0 0>, Passthrough Vector=<0 1 0>)
    SetAlignment=true
    SetFacing=true
   )
   &Color(Color=/VehicleCosmetics/FX/Niagara/Modules/MergeEmitters/VC_Dynamic_ColorBySpawnGroup(
     Color SG 0:LinearColor=$Particles.Color
     MaximumSpawnGroups:NiagaraInt32=2
     Color SG 1:LinearColor=$Particles.R1 Color
     Color SG 2:LinearColor=<rgba 1 1 1 1>
     Color SG 3:LinearColor=<rgba 1 1 1 1>
     Color SG 4:LinearColor=<rgba 1 1 1 1>
    ))
  ]
  particleUpdateScript [
   &ParticleState()
   &ScaleSpriteSize(
    Scale Factor=&MaskVector2dBySpawnGroup(
     Masked Vector2D=&Vector2DFromCurve(
      Vector2Curve=DataInterface() {propertyValues {
        XCurve {keys [
          {arriveTangent=0.2281, interpMode=RCIM_Cubic, leaveTangent=0.2281, tangentMode=RCTM_User, time=0, value=0.5}
          {arriveTangent=0.8667, interpMode=RCIM_Cubic, leaveTangent=0.8667, tangentMode=RCTM_User, time=1, value=1}
         ]}
        YCurve {keys [{arriveTangent=2.4739, interpMode=RCIM_Cubic, leaveTangent=2.4739, tangentMode=RCTM_User, time=0, value=1}, <at 0.5 2 RCIM_Cubic>]}
       }}
      ScaleCurve=&MakeCustomVector2DFromBool(Bool=$User.bIsFrontendPreview, True Vector2D=<-2 1.15>, False Vector2D=<-2 1>)
     )
     Passthrough Vector2D=&Vector2DFromCurve(
      Vector2Curve=DataInterface() {propertyValues {
        XCurve {keys [<at 0 0.5 RCIM_Cubic>,<at 1 1 RCIM_Cubic>]}
        YCurve {keys [{arriveTangent=2.4739, interpMode=RCIM_Cubic, leaveTangent=2.4739, tangentMode=RCTM_User, time=0, value=1}, <at 1 2 RCIM_Cubic>]}
       }}
      ScaleCurve=<-2 1>
     )
     Spawn Group Mask=0
    )
    Scale Sprite Size Mode=::NewEnumerator2
   )
   &SolveForcesAndVelocity(Clamp Velocity=false, Limit Acceleration=false)
   &SubUVAnimation(
    Random Start Frame=true
    "Sprite Renderer "=DataInterface() {propertyValues {SpriteRenderer=&self:NE_Source_Lightning001_0.NiagaraSpriteRendererProperties_1}}
    SubUV Animation Mode:ENiagara_SubUVLookupModeV2=::NewEnumerator1
    SubUV Curve=DataInterface() {propertyValues {Curve {keys [{interpMode=RCIM_Cubic, tangentMode=RCTM_User, time=0, value=0}, {interpMode=RCIM_Cubic, tangentMode=RCTM_User, time=0.1, value=4}]}}}
    UseEndFrame=false
    UseStartFrame=false
   ) {module=SubUVAnimation001}
   &SubUVAnimation(
    Output Alternate Renderer Binding=true
    Random Start Frame=true
    "Sprite Renderer "=DataInterface() {propertyValues {SpriteRenderer=&self:NE_Source_Lightning001_0.NiagaraSpriteRendererProperties_2}}
    Start Frame Offset:NiagaraInt32=&RandomRangeInt(Minimum=0, Maximum=5)
    SubUV Animation Mode:ENiagara_SubUVLookupModeV2=::NewEnumerator1
    SubUV Curve=DataInterface() {propertyValues {Curve {
       defaultValue=0
       keys [{interpMode=RCIM_Cubic, tangentMode=RCTM_User, time=0, value=0}, {interpMode=RCIM_Cubic, tangentMode=RCTM_User, time=0.1, value=8}]
      }}}
    UseEndFrame=false
    UseStartFrame=false
   )
   &ScaleColor(
    Scale Alpha=&MaskFloatBySpawnGroup(
     Masked Float=&FloatFromCurve(
      FloatCurve=DataInterface() {propertyValues {Curve {keys [
          {arriveTangent=-7.8312, interpMode=RCIM_Cubic, leaveTangent=-7.8312, tangentMode=RCTM_User, time=0, value=1}
          {interpMode=RCIM_Cubic, tangentMode=RCTM_User, time=0.4, value=0.01}
          <at 0.7 0 RCIM_Cubic>
         ]}}}
      Scale Curve=3
     )
     Passthrough Float=&FloatFromCurve(FloatCurve=DataInterface() {propertyValues {Curve {keys [
          {interpMode=RCIM_Cubic, tangentMode=RCTM_User, time=0, value=0.94}
          {interpMode=RCIM_Cubic, tangentMode=RCTM_User, time=0.3, value=0.5}
          {interpMode=RCIM_Cubic, tangentMode=RCTM_User, time=0.7, value=0}
         ]}}})
    )
    Scale RGB=&VectorFromFloat(Value=&MaskFloat(Mask Source=0, Mask Parameter=?, Mask=0, Masked Float=4, Passthrough Float=10))
    ScaleA=true
    ScaleRGB=true
   )
   &DynamicMaterialParameters(
    Index 0 Param 1=&FloatFromCurve(FloatCurve=DataInterface() {propertyValues {Curve {
        defaultValue=0
        keys [
         {interpMode=RCIM_Cubic, tangentMode=RCTM_User, time=0, value=0}
         {arriveTangent=1.6299, interpMode=RCIM_Cubic, leaveTangent=1.6299, tangentMode=RCTM_User, time=1, value=1}
        ]
       }}})
    Module.Index0Float0Write=true
    Module.Index0Float1Write=true
    Module.Index0Float2Write=true
    Module.Index0Float3Write=true
    Module.Index1Float0Write=true
    Module.Index1Float1Write=true
    Module.Index1Float2Write=true
    Module.Index1Float3Write=true
    Write Parameter Index 1=::NewEnumerator1
   )
  ]
  rendererClasses [&NiagaraSpriteRendererProperties]
  renderers [
   {rendererClass=&NiagaraSpriteRendererProperties}
   {rendererClass=&NiagaraSpriteRendererProperties, rendererIndex=1}
  ]
 }
}
summary {
 systemName=NS_Boost_Lightning_Optimized
 emitters [
  {emitterName=Lightning_Trail001, rendererClasses [&NiagaraRibbonRendererProperties]}
  {emitterName=NE_Source_Lightning001, rendererClasses [&NiagaraSpriteRendererProperties]}
  {emitterName=NE_Lightning_template001, rendererClasses [&NiagaraSpriteRendererProperties]}
 ]
 userVariables [
  User.BackLightPower {type=NiagaraInt32, defaultValue=1}
  User.bIsFrontendPreview {type=NiagaraBool}
  User.bIsPainted {type=NiagaraBool}
  User.bPreviewVehicleSpeed {type=NiagaraBool}
  User.ColorVariantID {type=NiagaraInt32}
  User.DTCarOffset {type=Vector3f, description="", defaultValue {struct=Vector3f, value=<0 0 0>}}
  User.GameModeID {type=NiagaraInt32}
  User.PaintedColor {type=LinearColor, description="", defaultValue {struct=LinearColor, value=<rgba 1 1 1 1>}}
  User.PaintedScalar {type=NiagaraFloat, defaultValue=1, description=""}
  User.PortCount {type=NiagaraInt32, defaultValue=2}
  User.TextSwtch {type=NiagaraInt32}
 ]
}
